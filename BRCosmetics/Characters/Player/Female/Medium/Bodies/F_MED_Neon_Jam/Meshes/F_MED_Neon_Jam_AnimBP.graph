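# Post-style secondary-motion AnimBP for the Neon Jam female medium cosmetic body, copying its pose from the attached parent mesh.
# A weighted constraint pins the phone master bone between pelvis, left thigh and thigh twist.
# Serial single-bone AnimDynamics nodes then add half-alpha, sprung, gravity-driven sway to phone, sleeves, shorts and straps.
# All but the phone simulate relative to root, and the event graph holds no logic.
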AnimBlueprint /BRCosmetics/Characters/Player/Female/Medium/Bodies/F_MED_Neon_Jam/Meshes/F_MED_Neon_Jam_AnimBP
target_skeleton=&/F_MED_Neon_Jam_Skeleton
graphs [
 AnimGraph {nodes [
   n1 = "Constraint - dyn_phone_MSTR (PPP)"!(
    ComponentPose=Local To Component(LocalPose=Copy Pose From Mesh() {bUseAttachedParent=true})
    Alpha=1
    ConstraintWeights_0=0.25
    ConstraintWeights_1:float=0.2
    ConstraintWeights_2:float=0.55
   ) -> (Pose) {
    alphaBoolBlend {blendOption=Linear}
    boneToModify {boneName=dyn_phone_MSTR}
    constraintSetup [{targetBone {boneName=pelvis}}, {targetBone {boneName=thigh_l}}, {targetBone {boneName=thigh_twist_01_l}}]
    constraintWeights [1,1,1]
   }
   n2 = Anim Dynamics - Bone: dyn_phone_1!(ComponentPose=@n1, ExternalForce="0,0,0") -> (Pose) {
    alpha=0.5
    angularDampingOverride=3
    angularSpringConstant=50
    bGravityOverrideInSimSpace=true
    bUseGravityOverride=true
    alphaBoolBlend {blendOption=Linear}
    boundBone {boneName=dyn_phone_1}
    componentAppliedLinearAccClamp=<5000 5000 5000>
    componentLinearAccScale=<0.2 0.2 0.55>
    componentLinearVelScale=<0.2 0.2 0.3>
    gravityOverride {z=-980}
    physicsBodyDefinitions [{
      sphereCollisionRadius=0
      boundBone {boneName=dyn_phone_1}
      boxExtents=<3 10 20>
      constraintSetup {angularLimitsMax=<10 2 0>, angularLimitsMin=<-10 -5 -0.1>, angularTarget=<1 0 0>}
      localJointOffset=<0 0 -2>
     }]
    relativeSpaceBone {boneName=root}
   }
   n3 = Anim Dynamics - Bone: dyn_sleeve_l_01!(ComponentPose=@n2, ExternalForce="0,0,0") -> (Pose) {
    alpha=0.5
    angularDampingOverride=3
    angularSpringConstant=50
    bGravityOverrideInSimSpace=true
    bUseGravityOverride=true
    simulationSpace=BoneRelative
    alphaBoolBlend {blendOption=Linear}
    boundBone {boneName=dyn_sleeve_l_01}
    componentAppliedLinearAccClamp=<5000 5000 5000>
    componentLinearAccScale=<0.2 0.2 0.55>
    componentLinearVelScale=<0.2 0.2 0.3>
    gravityOverride {z=-980}
    physicsBodyDefinitions [{
      sphereCollisionRadius=0
      boundBone {boneName=dyn_sleeve_l_01}
      boxExtents=<10 30 30>
      constraintSetup {angularLimitsMax=<0 10 0>, angularLimitsMin=<-0.1 -20 -0.1>, angularTarget=<1 0 0>}
      localJointOffset=<6 0 0>
     }]
    relativeSpaceBone {boneName=root}
   }
   n4 = Anim Dynamics - Bone: dyn_sleeve_r_01!(ComponentPose=@n3, ExternalForce="0,0,0") -> (Pose) {
    alpha=0.5
    angularDampingOverride=3
    angularSpringConstant=50
    bGravityOverrideInSimSpace=true
    bUseGravityOverride=true
    simulationSpace=BoneRelative
    alphaBoolBlend {blendOption=Linear}
    boundBone {boneName=dyn_sleeve_r_01}
    componentAppliedLinearAccClamp=<5000 5000 5000>
    componentLinearAccScale=<0.2 0.2 0.55>
    componentLinearVelScale=<0.2 0.2 0.3>
    gravityOverride {z=-980}
    physicsBodyDefinitions [{
      sphereCollisionRadius=0
      boundBone {boneName=dyn_sleeve_r_01}
      boxExtents=<10 30 30>
      constraintSetup {angularLimitsMax=<0 10 0>, angularLimitsMin=<-0.1 -20 -0.1>, angularTarget=<1 0 0>}
      localJointOffset=<6 0 0>
     }]
    relativeSpaceBone {boneName=root}
   }
   n5 = Anim Dynamics - Bone: dyn_sleeve_l_04!(ComponentPose=@n4, ExternalForce="0,0,0") -> (Pose) {
    alpha=0.5
    angularDampingOverride=3
    angularSpringConstant=50
    bGravityOverrideInSimSpace=true
    bUseGravityOverride=true
    simulationSpace=BoneRelative
    alphaBoolBlend {blendOption=Linear}
    boundBone {boneName=dyn_sleeve_l_04}
    componentAppliedLinearAccClamp=<5000 5000 5000>
    componentLinearAccScale=<0.2 0.2 0.55>
    componentLinearVelScale=<0.2 0.2 0.3>
    gravityOverride {z=-980}
    physicsBodyDefinitions [{
      sphereCollisionRadius=0
      boundBone {boneName=dyn_sleeve_l_04}
      boxExtents=<10 30 30>
      constraintSetup {angularLimitsMax=<0 10 0>, angularLimitsMin=<-0.1 -20 -0.1>, angularTarget=<1 0 0>}
      localJointOffset=<6 0 0>
     }]
    relativeSpaceBone {boneName=root}
   }
   n6 = Anim Dynamics - Bone: dyn_sleeve_r_04!(ComponentPose=@n5, ExternalForce="0,0,0") -> (Pose) {
    alpha=0.5
    angularDampingOverride=3
    angularSpringConstant=50
    bGravityOverrideInSimSpace=true
    bUseGravityOverride=true
    simulationSpace=BoneRelative
    alphaBoolBlend {blendOption=Linear}
    boundBone {boneName=dyn_sleeve_r_04}
    componentAppliedLinearAccClamp=<5000 5000 5000>
    componentLinearAccScale=<0.2 0.2 0.55>
    componentLinearVelScale=<0.2 0.2 0.3>
    gravityOverride {z=-980}
    physicsBodyDefinitions [{
      sphereCollisionRadius=0
      boundBone {boneName=dyn_sleeve_r_04}
      boxExtents=<10 30 30>
      constraintSetup {angularLimitsMax=<0 10 0>, angularLimitsMin=<-0.1 -20 -0.1>, angularTarget=<1 0 0>}
      localJointOffset=<6 0 0>
     }]
    relativeSpaceBone {boneName=root}
   }
   n7 = Anim Dynamics - Bone: dyn_sleeve_l_02!(ComponentPose=@n6, ExternalForce="0,0,0") -> (Pose) {
    alpha=0.5
    angularDampingOverride=3
    angularSpringConstant=50
    bGravityOverrideInSimSpace=true
    bUseGravityOverride=true
    simulationSpace=BoneRelative
    alphaBoolBlend {blendOption=Linear}
    boundBone {boneName=dyn_sleeve_l_02}
    componentAppliedLinearAccClamp=<5000 5000 5000>
    componentLinearAccScale=<0.2 0.2 0.55>
    componentLinearVelScale=<0.2 0.2 0.3>
    gravityOverride {z=-980}
    physicsBodyDefinitions [{
      sphereCollisionRadius=0
      boundBone {boneName=dyn_sleeve_l_02}
      boxExtents=<10 30 30>
      constraintSetup {angularLimitsMax=<0 -0.1 0>, angularLimitsMin=<-0.1 -20 -0.1>, angularTarget=<1 0 0>}
      localJointOffset=<6 0 0>
     }]
    relativeSpaceBone {boneName=root}
   }
   n8 = Anim Dynamics - Bone: dyn_sleeve_r_02!(ComponentPose=@n7, ExternalForce="0,0,0") -> (Pose) {
    alpha=0.5
    angularDampingOverride=3
    angularSpringConstant=50
    bGravityOverrideInSimSpace=true
    bUseGravityOverride=true
    simulationSpace=BoneRelative
    alphaBoolBlend {blendOption=Linear}
    boundBone {boneName=dyn_sleeve_r_02}
    componentAppliedLinearAccClamp=<5000 5000 5000>
    componentLinearAccScale=<0.2 0.2 0.55>
    componentLinearVelScale=<0.2 0.2 0.3>
    gravityOverride {z=-980}
    physicsBodyDefinitions [{
      sphereCollisionRadius=0
      boundBone {boneName=dyn_sleeve_r_02}
      boxExtents=<10 30 30>
      constraintSetup {angularLimitsMax=<0 10 0>, angularLimitsMin=<-0.1 -20 -0.1>, angularTarget=<1 0 0>}
      localJointOffset=<6 0 0>
     }]
    relativeSpaceBone {boneName=root}
   }
   n9 = Anim Dynamics - Bone: dyn_sleeve_l_03!(ComponentPose=@n8, ExternalForce="0,0,0") -> (Pose) {
    alpha=0.5
    angularDampingOverride=3
    angularSpringConstant=50
    bGravityOverrideInSimSpace=true
    bUseGravityOverride=true
    simulationSpace=BoneRelative
    alphaBoolBlend {blendOption=Linear}
    boundBone {boneName=dyn_sleeve_l_03}
    componentAppliedLinearAccClamp=<5000 5000 5000>
    componentLinearAccScale=<0.2 0.2 0.55>
    componentLinearVelScale=<0.2 0.2 0.3>
    gravityOverride {z=-980}
    physicsBodyDefinitions [{
      sphereCollisionRadius=0
      boundBone {boneName=dyn_sleeve_l_03}
      boxExtents=<10 30 30>
      constraintSetup {angularLimitsMax=<0 10 0>, angularLimitsMin=<-0.1 -20 -0.1>, angularTarget=<1 0 0>}
      localJointOffset=<6 0 0>
     }]
    relativeSpaceBone {boneName=root}
   }
   n10 = Anim Dynamics - Bone: dyn_sleeve_r_03!(ComponentPose=@n9, ExternalForce="0,0,0") -> (Pose) {
    alpha=0.5
    angularDampingOverride=3
    angularSpringConstant=50
    bGravityOverrideInSimSpace=true
    bUseGravityOverride=true
    simulationSpace=BoneRelative
    alphaBoolBlend {blendOption=Linear}
    boundBone {boneName=dyn_sleeve_r_03}
    componentAppliedLinearAccClamp=<5000 5000 5000>
    componentLinearAccScale=<0.2 0.2 0.55>
    componentLinearVelScale=<0.2 0.2 0.3>
    gravityOverride {z=-980}
    physicsBodyDefinitions [{
      sphereCollisionRadius=0
      boundBone {boneName=dyn_sleeve_r_03}
      boxExtents=<10 30 30>
      constraintSetup {angularLimitsMax=<0 10 0>, angularLimitsMin=<-0.1 -20 -0.1>, angularTarget=<1 0 0>}
      localJointOffset=<6 0 0>
     }]
    relativeSpaceBone {boneName=root}
   }
   n11 = Anim Dynamics - Bone: dyn_shorts_l_01!(ComponentPose=@n10, ExternalForce="0,0,0") -> (Pose) {
    alpha=0.5
    angularDampingOverride=3
    angularSpringConstant=50
    bGravityOverrideInSimSpace=true
    bUseGravityOverride=true
    simulationSpace=BoneRelative
    alphaBoolBlend {blendOption=Linear}
    boundBone {boneName=dyn_shorts_l_01}
    componentAppliedLinearAccClamp=<5000 5000 5000>
    componentLinearAccScale=<0.1 0.1 0.2>
    componentLinearVelScale=<0.2 0.2 0.4>
    gravityOverride {z=-980}
    physicsBodyDefinitions [{
      sphereCollisionRadius=0
      boundBone {boneName=dyn_shorts_l_01}
      boxExtents=<15 12 12>
      constraintSetup {angularLimitsMax=<0 20 0>, angularLimitsMin=<-0.1 -60 -0.1>, angularTarget=<1 0 0>}
      localJointOffset=<3 0 0>
     }]
    relativeSpaceBone {boneName=root}
   }
   n12 = Anim Dynamics - Bone: dyn_shorts_r_01!(ComponentPose=@n11, ExternalForce="0,0,0") -> (Pose) {
    alpha=0.5
    angularDampingOverride=3
    angularSpringConstant=50
    bGravityOverrideInSimSpace=true
    bUseGravityOverride=true
    simulationSpace=BoneRelative
    alphaBoolBlend {blendOption=Linear}
    boundBone {boneName=dyn_shorts_r_01}
    componentAppliedLinearAccClamp=<5000 5000 5000>
    componentLinearAccScale=<0.1 0.1 0.2>
    componentLinearVelScale=<0.2 0.2 0.4>
    gravityOverride {z=-980}
    physicsBodyDefinitions [{
      sphereCollisionRadius=0
      boundBone {boneName=dyn_shorts_r_01}
      boxExtents=<15 12 12>
      constraintSetup {angularLimitsMax=<0 20 0>, angularLimitsMin=<-0.1 -60 -0.1>, angularTarget=<1 0 0>}
      localJointOffset=<3 0 0>
     }]
    relativeSpaceBone {boneName=root}
   }
   n13 = Anim Dynamics - Bone: dyn_shorts_l_04!(ComponentPose=@n12, ExternalForce="0,0,0") -> (Pose) {
    alpha=0.5
    angularDampingOverride=3
    angularSpringConstant=50
    bGravityOverrideInSimSpace=true
    bUseGravityOverride=true
    simulationSpace=BoneRelative
    alphaBoolBlend {blendOption=Linear}
    boundBone {boneName=dyn_shorts_l_04}
    componentAppliedLinearAccClamp=<5000 5000 5000>
    componentLinearAccScale=<0.1 0.1 0.2>
    componentLinearVelScale=<0.2 0.2 0.4>
    gravityOverride {z=-980}
    physicsBodyDefinitions [{
      sphereCollisionRadius=0
      boundBone {boneName=dyn_shorts_l_04}
      boxExtents=<15 12 12>
      constraintSetup {angularLimitsMax=<0 20 0>, angularLimitsMin=<-0.1 -60 -0.1>, angularTarget=<1 0 0>}
      localJointOffset=<3 0 0>
     }]
    relativeSpaceBone {boneName=root}
   }
   n14 = Anim Dynamics - Bone: dyn_shorts_r_04!(ComponentPose=@n13, ExternalForce="0,0,0") -> (Pose) {
    alpha=0.5
    angularDampingOverride=3
    angularSpringConstant=50
    bGravityOverrideInSimSpace=true
    bUseGravityOverride=true
    simulationSpace=BoneRelative
    alphaBoolBlend {blendOption=Linear}
    boundBone {boneName=dyn_shorts_r_04}
    componentAppliedLinearAccClamp=<5000 5000 5000>
    componentLinearAccScale=<0.1 0.1 0.2>
    componentLinearVelScale=<0.2 0.2 0.4>
    gravityOverride {z=-980}
    physicsBodyDefinitions [{
      sphereCollisionRadius=0
      boundBone {boneName=dyn_shorts_r_04}
      boxExtents=<15 12 12>
      constraintSetup {angularLimitsMax=<0 20 0>, angularLimitsMin=<-0.1 -60 -0.1>, angularTarget=<1 0 0>}
      localJointOffset=<3 0 0>
     }]
    relativeSpaceBone {boneName=root}
   }
   n15 = Anim Dynamics - Bone: dyn_shorts_l_02!(ComponentPose=@n14, ExternalForce="0,0,0") -> (Pose) {
    alpha=0.5
    angularDampingOverride=3
    angularSpringConstant=50
    bGravityOverrideInSimSpace=true
    bUseGravityOverride=true
    simulationSpace=BoneRelative
    alphaBoolBlend {blendOption=Linear}
    boundBone {boneName=dyn_shorts_l_02}
    componentAppliedLinearAccClamp=<5000 5000 5000>
    componentLinearAccScale=<0.1 0.1 0.2>
    componentLinearVelScale=<0.2 0.2 0.4>
    gravityOverride {z=-980}
    physicsBodyDefinitions [{
      sphereCollisionRadius=0
      boundBone {boneName=dyn_shorts_l_02}
      boxExtents=<15 12 12>
      constraintSetup {angularLimitsMax=<0 20 0>, angularLimitsMin=<-0.1 -60 -0.1>, angularTarget=<1 0 0>}
      localJointOffset=<3 0 0>
     }]
    relativeSpaceBone {boneName=root}
   }
   n16 = Anim Dynamics - Bone: dyn_shorts_r_02!(ComponentPose=@n15, ExternalForce="0,0,0") -> (Pose) {
    alpha=0.5
    angularDampingOverride=3
    angularSpringConstant=50
    bGravityOverrideInSimSpace=true
    bUseGravityOverride=true
    simulationSpace=BoneRelative
    alphaBoolBlend {blendOption=Linear}
    boundBone {boneName=dyn_shorts_r_02}
    componentAppliedLinearAccClamp=<5000 5000 5000>
    componentLinearAccScale=<0.1 0.1 0.2>
    componentLinearVelScale=<0.2 0.2 0.4>
    gravityOverride {z=-980}
    physicsBodyDefinitions [{
      sphereCollisionRadius=0
      boundBone {boneName=dyn_shorts_r_02}
      boxExtents=<15 12 12>
      constraintSetup {angularLimitsMax=<0 20 0>, angularLimitsMin=<-0.1 -60 -0.1>, angularTarget=<1 0 0>}
      localJointOffset=<3 0 0>
     }]
    relativeSpaceBone {boneName=root}
   }
   n17 = Anim Dynamics - Bone: dyn_shorts_l_03!(ComponentPose=@n16, ExternalForce="0,0,0") -> (Pose) {
    alpha=0.5
    angularDampingOverride=3
    angularSpringConstant=50
    bGravityOverrideInSimSpace=true
    bUseGravityOverride=true
    simulationSpace=BoneRelative
    alphaBoolBlend {blendOption=Linear}
    boundBone {boneName=dyn_shorts_l_03}
    componentAppliedLinearAccClamp=<5000 5000 5000>
    componentLinearAccScale=<0.1 0.1 0.2>
    componentLinearVelScale=<0.2 0.2 0.4>
    gravityOverride {z=-980}
    physicsBodyDefinitions [{
      sphereCollisionRadius=0
      boundBone {boneName=dyn_shorts_l_03}
      boxExtents=<15 12 12>
      constraintSetup {angularLimitsMax=<0 20 0>, angularLimitsMin=<-0.1 -60 -0.1>, angularTarget=<1 0 0>}
      localJointOffset=<3 0 0>
     }]
    relativeSpaceBone {boneName=root}
   }
   n18 = Anim Dynamics - Bone: dyn_shorts_r_03!(ComponentPose=@n17, ExternalForce="0,0,0") -> (Pose) {
    alpha=0.5
    angularDampingOverride=3
    angularSpringConstant=50
    bGravityOverrideInSimSpace=true
    bUseGravityOverride=true
    simulationSpace=BoneRelative
    alphaBoolBlend {blendOption=Linear}
    boundBone {boneName=dyn_shorts_r_03}
    componentAppliedLinearAccClamp=<5000 5000 5000>
    componentLinearAccScale=<0.1 0.1 0.2>
    componentLinearVelScale=<0.2 0.2 0.4>
    gravityOverride {z=-980}
    physicsBodyDefinitions [{
      sphereCollisionRadius=0
      boundBone {boneName=dyn_shorts_r_03}
      boxExtents=<15 12 12>
      constraintSetup {angularLimitsMax=<0 20 0>, angularLimitsMin=<-0.1 -60 -0.1>, angularTarget=<1 0 0>}
      localJointOffset=<3 0 0>
     }]
    relativeSpaceBone {boneName=root}
   }
   n19 = Anim Dynamics - Bone: dyn_strap_l_01!(ComponentPose=@n18, ExternalForce="0,0,0") -> (Pose) {
    alpha=0.5
    angularDampingOverride=3
    angularSpringConstant=50
    bGravityOverrideInSimSpace=true
    bUseGravityOverride=true
    simulationSpace=BoneRelative
    alphaBoolBlend {blendOption=Linear}
    boundBone {boneName=dyn_strap_l_01}
    componentAppliedLinearAccClamp=<5000 5000 5000>
    componentLinearAccScale=<0.2 0.2 0.55>
    componentLinearVelScale=<0.2 0.2 0.3>
    gravityOverride {z=-980}
    physicsBodyDefinitions [{
      sphereCollisionRadius=0
      boundBone {boneName=dyn_strap_l_01}
      boxExtents=<20 10 20>
      constraintSetup {angularLimitsMax=<0 5 30>, angularLimitsMin=<-0.1 -5 -30>, angularTarget=<1 0 0>}
      localJointOffset=<8 0 0>
     }]
    relativeSpaceBone {boneName=root}
   }
   n20 = Anim Dynamics - Bone: dyn_strap_r_01!(ComponentPose=@n19, ExternalForce="0,0,0") -> (Pose) {
    alpha=0.5
    angularDampingOverride=3
    angularSpringConstant=50
    bGravityOverrideInSimSpace=true
    bUseGravityOverride=true
    simulationSpace=BoneRelative
    alphaBoolBlend {blendOption=Linear}
    boundBone {boneName=dyn_strap_r_01}
    componentAppliedLinearAccClamp=<5000 5000 5000>
    componentLinearAccScale=<0.2 0.2 0.55>
    componentLinearVelScale=<0.2 0.2 0.3>
    gravityOverride {z=-980}
    physicsBodyDefinitions [{
      sphereCollisionRadius=0
      boundBone {boneName=dyn_strap_r_01}
      boxExtents=<20 10 20>
      constraintSetup {angularLimitsMax=<0 5 30>, angularLimitsMin=<-0.1 -5 -30>, angularTarget=<1 0 0>}
      localJointOffset=<8 0 0>
     }]
    relativeSpaceBone {boneName=root}
   }
   n21 = Anim Dynamics - Bone: dyn_strap_l_02!(ComponentPose=@n20, ExternalForce="0,0,0") -> (Pose) {
    alpha=0.5
    angularDampingOverride=3
    angularSpringConstant=50
    bGravityOverrideInSimSpace=true
    bUseGravityOverride=true
    simulationSpace=BoneRelative
    alphaBoolBlend {blendOption=Linear}
    boundBone {boneName=dyn_strap_l_02}
    componentAppliedLinearAccClamp=<5000 5000 5000>
    componentLinearAccScale=<0.2 0.2 0.55>
    componentLinearVelScale=<0.2 0.2 0.3>
    gravityOverride {z=-980}
    physicsBodyDefinitions [{
      sphereCollisionRadius=0
      boundBone {boneName=dyn_strap_l_02}
      boxExtents=<20 10 20>
      constraintSetup {angularLimitsMax=<0 5 30>, angularLimitsMin=<-0.1 -5 -30>, angularTarget=<1 0 0>}
      localJointOffset=<8 0 0>
     }]
    relativeSpaceBone {boneName=root}
   }
   n22 = Anim Dynamics - Bone: dyn_strap_r_02!(ComponentPose=@n21, ExternalForce="0,0,0") -> (Pose) {
    alpha=0.5
    angularDampingOverride=3
    angularSpringConstant=50
    bGravityOverrideInSimSpace=true
    bUseGravityOverride=true
    simulationSpace=BoneRelative
    alphaBoolBlend {blendOption=Linear}
    boundBone {boneName=dyn_strap_r_02}
    componentAppliedLinearAccClamp=<5000 5000 5000>
    componentLinearAccScale=<0.2 0.2 0.55>
    componentLinearVelScale=<0.2 0.2 0.3>
    gravityOverride {z=-980}
    physicsBodyDefinitions [{
      sphereCollisionRadius=0
      boundBone {boneName=dyn_strap_r_02}
      boxExtents=<20 10 20>
      constraintSetup {angularLimitsMax=<0 5 30>, angularLimitsMin=<-0.1 -5 -30>, angularTarget=<1 0 0>}
      localJointOffset=<8 0 0>
     }]
    relativeSpaceBone {boneName=root}
   }
   Output Pose(Result=Component To Local(ComponentPose=@n22)) {layerGroup=None}
  ]}
 EventGraph {nodes [Event BlueprintUpdateAnimation()]}
]
variables []
