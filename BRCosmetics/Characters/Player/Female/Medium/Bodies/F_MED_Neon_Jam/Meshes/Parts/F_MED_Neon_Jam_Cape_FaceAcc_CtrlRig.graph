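# Control Rig for the F_MED_Neon_Jam Cape_FaceAcc part that adjusts how the dreadlock hair sits relative to the upper back.
# It sets const_hair_aim to a fixed local transform under spine_05, then half-weight aims head_offset at that bone, with secondary aim disabled.
# The lettered dread chains inherit this through head_offset, while the front dread hangs directly from head.
# The isWearingCape and isWearingPack variables are declared but unused by the graph.
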ControlRigBlueprint /BRCosmetics/Characters/Player/Female/Medium/Bodies/F_MED_Neon_Jam/Meshes/Parts/F_MED_Neon_Jam_Cape_FaceAcc_CtrlRig
graphs [
 RigVMFunctionLibrary {links [], nodes []}
 RigVMModel {nodes [
   RigUnit_BeginExecution = &RigUnit_BeginExecution()
   &RigUnit_SetTransform(
    Item=(Type=Bone,Name="const_hair_aim")
    Space=LocalSpace
    Value=(Rotation=(X=0,Y=0,Z=0,W=1),Translation=(X=0,Y=2,Z=0),Scale3D=(X=1,Y=1,Z=1))
    Weight=1
    ExecutePin=@RigUnit_BeginExecution.ExecutePin
   )
   > /Script/ControlRig.RigUnit_AimItem(
    DebugSettings=(Scale=10,WorldOffset=(Rotation=(X=0,Y=0,Z=0,W=1),Translation=(X=0,Y=0,Z=0),Scale3D=(X=1,Y=1,Z=1)))
    Item=(Type=Bone,Name="head_offset")
    Primary=(Weight=0.5,Axis=(X=-1,Y=1,Z=0),Target=(X=0,Y=0,Z=0),Kind=Location,Space=(Type=Bone,Name="const_hair_aim"))
    Secondary=(Weight=0,Axis=(X=0,Y=0,Z=0),Target=(X=1,Y=0,Z=0),Kind=Direction,Space=(Type=Bone,Name="head"))
    Weight=1
   ) {is=struct RigVMUnitNode}
  ]}
]
hierarchy [
 root {loc=<0 0 0>, rot=<0 0 0 1>}
 pelvis {parent=root, loc=<0 -0.5301 98.0571>, rot=<0.0127 0.707 -0.0127 -0.707>}
 spine_01 {parent=pelvis, loc=<6.1208 -0.9328 0>, rot=<0 0 0.034 -0.9994>}
 spine_02 {parent=spine_01, loc=<6.5996 -0.2095 0>, rot=<0 0 0.0238 -0.9997>}
 spine_03 {parent=spine_02, loc=<6.4914 -0.1237 0>, rot=<0 0 0.0149 -0.9999>}
 spine_04 {parent=spine_03, loc=<6.9598 0.0794 0>, rot=<0 0 -0.0132 -0.9999>}
 spine_05 {parent=spine_04, loc=<6.9987 0 0>, rot=<0 0 -0.0015 -1>}
 clavicle_l {parent=spine_05, loc=<5.4831 -0.1053 -3.2946>, rot=<0.0465 -0.7289 -0.009 -0.6829>}
 upperarm_l {parent=clavicle_l, loc=<9.8658 -1.2136 -0.4105>, rot=<0.0164 -0.3245 -0.0269 -0.9454>}
 neck_01 {parent=spine_05, loc=<11.3138 0.1529 0>, rot=<0 0 -0.2003 -0.9797>}
 neck_02 {parent=neck_01, loc=<4.5627 0 0>, rot=<0 0 0.0504 -0.9987>}
 head {parent=neck_02, loc=<5.3914 0 0>, rot=<0 0 0.115 -0.9934>}
 head_offset {parent=head, loc=<13.2166 -2.4544 0>, rot=<0 0 0 1>}
 dyn_dread_base_a {parent=head_offset, loc=<1.6879 -8.5936 6.9101>, rot=<0.1211 -0.8602 -0.4376 0.2324>}
 dyn_dread_parent_a {parent=dyn_dread_base_a, loc=<0 0 0>, rot=<0 0 0 1>}
 dyn_dread_a_01 {parent=dyn_dread_parent_a, loc=<11.6945 1.3985 3.1417>, rot=<0.1758 0.2147 0.0138 -0.9606>}
 dyn_dread_a_02 {parent=dyn_dread_a_01, loc=<10.2975 -0.0372 0.1656>, rot=<-0.0181 0.0046 0.0224 -0.9996>}
 dyn_dread_a_03 {parent=dyn_dread_a_02, loc=<13.2112 0.5607 -0.9966>, rot=<0 0 0 1>}
 dyn_dread_base_b {parent=head_offset, loc=<3.6061 -9.9597 -5.267>, rot=<-0.1866 0.5581 0.7976 -0.1323>}
 dyn_dread_parent_b {parent=dyn_dread_base_b, loc=<0 0 0>, rot=<0 0 0 1>}
 dyn_dread_b_01 {parent=dyn_dread_parent_b, loc=<15.9133 -1.3823 3.5524>, rot=<-0.0234 0.1926 0.024 -0.9807>}
 dyn_dread_b_02 {parent=dyn_dread_b_01, loc=<17.5791 0.3376 0.5466>, rot=<0.0361 -0.0102 -0.0265 -0.9989>}
 dyn_dread_base_c {parent=head_offset, loc=<0.9832 -6.3732 0.9248>, rot=<0.2042 -0.7594 -0.5512 0.279>}
 dyn_dread_parent_c {parent=dyn_dread_base_c, loc=<0 0 0>, rot=<0 0 0 1>}
 dyn_dread_c_01 {parent=dyn_dread_parent_c, loc=<16.7411 1.3721 9.3373>, rot=<0.0534 0.3082 -0.0449 -0.9488>}
 dyn_dread_c_02 {parent=dyn_dread_c_01, loc=<15.5198 0.4209 -0.1441>, rot=<-0.0527 -0.0041 0.0026 -0.9986>}
 dyn_dread_base_d {parent=head_offset, loc=<1.7792 -6.7333 -5.8347>, rot=<-0.3846 0.3814 0.8206 -0.182>}
 dyn_dread_parent_d {parent=dyn_dread_base_d, loc=<0 0 0>, rot=<0 0 0 1>}
 dyn_dread_d_01 {parent=dyn_dread_parent_d, loc=<8.9664 -0.303 3.7312>, rot=<-0.0169 0.3916 0.0363 -0.9193>}
 dyn_dread_d_02 {parent=dyn_dread_d_01, loc=<12.8866 0.1205 0.1849>, rot=<0.0006 0.0041 -0.0216 -0.9998>}
 dyn_dread_d_03 {parent=dyn_dread_d_02, loc=<12.4163 -0.0948 -0.0302>, rot=<0 0 0 1>}
 dyn_dread_base_e {parent=head_offset, loc=<5.6561 -9.5559 -1.4573>, rot=<-0.3141 0.611 0.667 -0.2883>}
 dyn_dread_parent_e {parent=dyn_dread_base_e, loc=<0 0 0>, rot=<0 0 0 1>}
 dyn_dread_e_01 {parent=dyn_dread_parent_e, loc=<10.3243 -0.2326 4.9924>, rot=<0.0229 0.3582 0.0127 -0.9333>}
 dyn_dread_e_02 {parent=dyn_dread_e_01, loc=<14.5423 -0.1056 1.6508>, rot=<0 0 0 1>}
 dyn_dread_e_03 {parent=dyn_dread_e_02, loc=<13.6862 -0.2945 0.4832>, rot=<0 0 0 1>}
 dyn_dread_base_f {parent=head_offset, loc=<2.8499 -11.6296 4.2496>, rot=<0.0691 -0.8169 -0.5298 0.2173>}
 dyn_dread_parent_f {parent=dyn_dread_base_f, loc=<0 0 0>, rot=<0 0 0 1>}
 dyn_dread_f_01 {parent=dyn_dread_parent_f, loc=<16.284 -0.5561 5.0393>, rot=<0.0849 0.21 0.0806 -0.9707>}
 dyn_dread_f_02 {parent=dyn_dread_f_01, loc=<16.4443 0.145 -1.9087>, rot=<0.0349 0 0 -0.9994>}
 dyn_dread_base_g {parent=head_offset, loc=<-1.2927 -8.4848 -3.8287>, rot=<-0.1375 0.6304 0.7514 -0.138>}
 dyn_dread_parent_g {parent=dyn_dread_base_g, loc=<0 0 0>, rot=<0 0 0 1>}
 dyn_dread_g_01 {parent=dyn_dread_parent_g, loc=<16.532 -1.021 5.1697>, rot=<-0.0067 0.1405 0.0012 -0.9901>}
 dyn_dread_g_02 {parent=dyn_dread_g_01, loc=<15.9068 -0.2903 0.3051>, rot=<0 0 0 1>}
 dyn_dread_base_h {parent=head_offset, loc=<6.3508 -10.5531 0.9569>, rot=<0.3208 -0.6588 -0.5723 0.3683>}
 dyn_dread_parent_h {parent=dyn_dread_base_h, loc=<0 0 0>, rot=<0 0 0 1>}
 dyn_dread_h_01 {parent=dyn_dread_parent_h, loc=<7.6182 0.5158 4.5667>, rot=<-0.0153 0.4847 -0.0085 -0.8745>}
 dyn_dread_h_02 {parent=dyn_dread_h_01, loc=<13.833 0.4675 -0.1063>, rot=<0 -0.0698 0 -0.9976>}
 dyn_dread_h_03 {parent=dyn_dread_h_02, loc=<15.3232 -0.6718 0.6434>, rot=<0 0 0 1>}
 dyn_front_dread_base {parent=head, loc=<9.9716 6.8182 3.7831>, rot=<0.273 0.7155 -0.382 -0.5174>}
 dyn_front_dread_parent {parent=dyn_front_dread_base, loc=<0 0 0>, rot=<0 0 0 1>}
 dyn_front_dread_01 {parent=dyn_front_dread_parent, loc=<3.2542 0.0588 2.618>, rot=<-0.1879 0.5438 -0.1527 -0.8035>}
 dyn_front_dread_02 {parent=dyn_front_dread_01, loc=<7.9023 -0.1213 0.3089>, rot=<0 0.0346 0.0049 -0.9994>}
 dyn_front_dread_03 {parent=dyn_front_dread_02, loc=<8.8736 0.0816 0.0468>, rot=<0 0 0 1>}
 faceAttach {parent=head, loc=<0.6201 0.9296 0>, rot=<0.0029 -0.7071 -0.0029 -0.7071>}
 C_nose_bridge {parent=faceAttach, loc=<-0.0025 9.0629 0.0077>, rot=<-0.2218 0 0 -0.9751>}
 clavicle_r {parent=spine_05, loc=<5.4831 -0.1053 3.2946>, rot=<0.7289 0.0465 -0.6829 0.009>}
 upperarm_r {parent=clavicle_r, loc=<-9.8658 1.2136 0.4105>, rot=<0.0164 -0.3245 -0.0269 -0.9454>}
 const_hair_aim {parent=spine_05, loc=<-0.0998 -7.1089 0>, rot=<-0.0284 0.7065 0.7065 -0.0284>}
]
variables [
 isWearingCape {type=bool, default=False}
 isWearingPack {type=bool, default=False}
]
