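# Physics asset for the front-end face accessory mesh of the F_MED_Neon_Jam female cosmetic, defining dynamic dreadlock bones.
# Eight dread chains (a–h) and a front chain each run from a small base sphere through capsule segments, with tight twist and asymmetric swing limits.
# Capsules also cover the head, upper spine, right clavicle and both upper arms, with angularly free constraints linking the arms to spine_05 and clavicle_r.
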PhysicsAsset /BRCosmetics/Characters/Player/Female/Medium/Bodies/F_MED_Neon_Jam/Meshes/Parts/F_MED_Neon_Jam_FrontEnd_FaceAcc_Physics
bodies {
 clavicle_r [{shapeType=Capsule, length=5.8125, radius=9, shapeName=clavicle_r_capsule, center=<-8.7096 2.4584 3.1033>, rotation=<rot 90 0 0>}]
 dyn_dread_a_01 [{shapeType=Capsule, length=9.2004, radius=3, shapeName=dyn_dread_a_01_capsule, center=<5.4464 0 0>, rotation=<rot 90 0 0>}]
 dyn_dread_a_02 [{shapeType=Capsule, length=9.9504, radius=3, shapeName=dyn_dread_a_02_capsule, center=<5.4409 0 0>, rotation=<rot 90 0 0>}]
 dyn_dread_a_03 [{shapeType=Capsule, length=10.1045, radius=3.7886, shapeName=dyn_dread_a_03_capsule, center=<7.6652 0 0>, rotation=<rot 90 0 0>}]
 dyn_dread_b_01 [{shapeType=Capsule, length=11.2004, radius=3.25, shapeName=dyn_dread_b_01_capsule, center=<7.221 0 0>, rotation=<rot 90 0 0>}]
 dyn_dread_b_02 [{shapeType=Capsule, length=18.7004, radius=3, shapeName=dyn_dread_b_02_capsule, center=<8.7726 0 0>, rotation=<rot 90 0 0>}]
 dyn_dread_base_a [{shapeType=Sphere, radius=0.5161, shapeName=dyn_dread_base_a_sphere}]
 dyn_dread_base_b [{shapeType=Sphere, radius=0.5161, shapeName=dyn_dread_base_b_sphere}]
 dyn_dread_base_c [{shapeType=Sphere, radius=0.5161, shapeName=dyn_dread_base_c_sphere}]
 dyn_dread_base_d [{shapeType=Sphere, radius=0.5161, shapeName=dyn_dread_base_d_sphere}]
 dyn_dread_base_e [{shapeType=Sphere, radius=0.5161, shapeName=dyn_dread_base_e_sphere}]
 dyn_dread_base_f [{shapeType=Sphere, radius=0.5161, shapeName=dyn_dread_base_f_sphere}]
 dyn_dread_base_g [{shapeType=Sphere, radius=0.5, shapeName=dyn_dread_base_g_sphere}]
 dyn_dread_base_h [{shapeType=Sphere, radius=0.5, shapeName=dyn_dread_base_h_sphere}]
 dyn_dread_c_01 [{shapeType=Capsule, length=8.5442, radius=4.9062, shapeName=dyn_dread_c_01_capsule, center=<6.6296 0 -1.8>, rotation=<rot 90 0 0>}]
 dyn_dread_c_02 [{shapeType=Capsule, length=11.2317, radius=6.2188, shapeName=dyn_dread_c_02_capsule, center=<8.7167 0 -2.9>, rotation=<rot 90 0 0>}]
 dyn_dread_d_01 [{shapeType=Capsule, length=8.2004, radius=3.5, shapeName=dyn_dread_d_01_capsule, center=<6.0978 0 0>, rotation=<rot 90 0 0>}]
 dyn_dread_d_02 [{shapeType=Capsule, length=8.9504, radius=3.5, shapeName=dyn_dread_d_02_capsule, center=<6.5248 0 0>, rotation=<rot 90 0 0>}]
 dyn_dread_d_03 [{shapeType=Capsule, length=14.5675, radius=3.0589, shapeName=dyn_dread_d_03_capsule, center=<7.743 0 0>, rotation=<rot 90 0 0>}]
 dyn_dread_e_01 [{shapeType=Capsule, length=7.2004, radius=3.25, shapeName=dyn_dread_e_01_capsule, center=<5.7068 0 0>, rotation=<rot 90 0 0>}]
 dyn_dread_e_02 [{shapeType=Capsule, length=9.2004, radius=3, shapeName=dyn_dread_e_02_capsule, center=<6.1731 0 0>, rotation=<rot 90 0 0>}]
 dyn_dread_e_03 [{
   shapeType=Capsule
   length=11.6737
   radius=2.9715
   shapeName=dyn_dread_e_03_capsule
   center=<7.525 0.0921 1.0581>
   rotation=<rot 90 179.7762 176.9637>
  }]
 dyn_dread_f_01 [{shapeType=Capsule, length=13.2004, radius=3.5, shapeName=dyn_dread_f_01_capsule, center=<9.1943 0 0>, rotation=<rot 90 -180 180>}]
 dyn_dread_f_02 [{shapeType=Capsule, length=14.4504, radius=2.75, shapeName=dyn_dread_f_02_capsule, center=<9.5294 0 0>, rotation=<rot 90 0 0>}]
 dyn_dread_g_01 [{shapeType=Capsule, length=8.5578, radius=2.8225, shapeName=dyn_dread_g_01_capsule, center=<5.3378 0 0>, rotation=<rot 90 0 0>}]
 dyn_dread_g_02 [{shapeType=Capsule, length=12.4568, radius=3.0434, shapeName=dyn_dread_g_02_capsule, center=<5.7652 -0.713 0.7808>, rotation=<rot 90 0 0>}]
 dyn_dread_h_01 [{shapeType=Capsule, length=18.2693, radius=2.4779, shapeName=dyn_dread_h_01_capsule, center=<9.5143 0 0>, rotation=<rot 90 0 0>}]
 dyn_dread_h_02 [{
   shapeType=Capsule
   length=14.9499
   radius=1.8436
   shapeName=dyn_dread_h_02_capsule
   center=<9.2463 -0.3035 0>
   rotation=<rot 90 180 -177.1875>
  }]
 dyn_dread_h_03 [{
   shapeType=Capsule
   length=14.9499
   radius=2.3436
   shapeName=dyn_dread_h_03_capsule
   center=<8.2609 0.5799 0>
   rotation=<rot 90 178.5942 178.5942>
  }]
 dyn_dread_parent_a [{
   shapeType=Capsule
   length=9.9504
   radius=3.5
   shapeName=dyn_dread_parent_a_capsule
   center=<6.432 0 1.3316>
   rotation=<rot 81.5625 179.9999 179.9999>
  }]
 dyn_dread_parent_b [{shapeType=Capsule, length=11.4504, radius=2, shapeName=dyn_dread_parent_b_capsule, center=<6.6296 0 0>, rotation=<rot 90 0 0>}]
 dyn_dread_parent_c [{
   shapeType=Capsule
   length=10.9504
   radius=2.5
   shapeName=dyn_dread_parent_c_capsule
   center=<7.1395 0 1.5641>
   rotation=<rot 59.0624 180 180>
  }]
 dyn_dread_parent_d [{
   shapeType=Capsule
   length=6.2004
   radius=2.5
   shapeName=dyn_dread_parent_d_capsule
   center=<4.5511 0 1.2032>
   rotation=<rot 64.6874 -179.9999 -179.9998>
  }]
 dyn_dread_parent_e [{
   shapeType=Capsule
   length=7.4504
   radius=3.5
   shapeName=dyn_dread_parent_e_capsule
   center=<6.5207 0 2.249>
   rotation=<rot 70.3121 -179.9999 -179.9999>
  }]
 dyn_dread_parent_f [{
   shapeType=Capsule
   length=13.4504
   radius=2.75
   shapeName=dyn_dread_parent_f_capsule
   center=<8.4213 0 2.3056>
   rotation=<rot 75.9374 179.9999 179.9999>
  }]
 dyn_dread_parent_g [{shapeType=Capsule, length=7.6168, radius=2.7357, shapeName=dyn_dread_parent_g_capsule, center=<5.3378 0 0>, rotation=<rot 90 0 0>}]
 dyn_dread_parent_h [{
   shapeType=Capsule
   length=9.3547
   radius=2.3152
   shapeName=dyn_dread_parent_h_capsule
   center=<5.314 0 3.0014>
   rotation=<rot 61.875 180 180>
  }]
 dyn_front_dread_01 [{shapeType=Capsule, length=8.4321, radius=2, shapeName=dyn_front_dread_01_capsule, center=<2.8769 0 0>, rotation=<rot 90 0 0>}]
 dyn_front_dread_02 [{shapeType=Capsule, length=8.4321, radius=2, shapeName=dyn_front_dread_02_capsule, center=<2.8769 0 0>, rotation=<rot 90 0 0>}]
 dyn_front_dread_03 [{shapeType=Capsule, length=8.4321, radius=2, shapeName=dyn_front_dread_03_capsule, center=<2.8769 0 0>, rotation=<rot 90 0 0>}]
 dyn_front_dread_base [{shapeType=Sphere, radius=0.5161, shapeName=dyn_front_dread_base_sphere}]
 dyn_front_dread_parent [{
   shapeType=Capsule
   length=3.855
   radius=1.5
   shapeName=dyn_front_dread_parent_capsule
   center=<1.949 0 1.309>
   rotation=<rot 60.8517 180 180>
  }]
 head [{shapeType=Capsule, length=12.4868, radius=6.7651, shapeName=head_capsule, center=<0 -0.9105 0>, rotation=<rot 90 0 0>}]
 head_offset [{shapeType=Sphere, radius=0.5, shapeName=head_offset_sphere}]
 spine_04 [{shapeType=Capsule, length=23.0796, radius=11.4359, shapeName=spine_04_capsule, center=<-5.5559 2.44 0>}]
 spine_05 [{shapeType=Capsule, length=22.0323, radius=11.7893, shapeName=spine_05_capsule, center=<2.6846 3.846 0>}]
 upperarm_l [{shapeType=Capsule, length=32, radius=8, shapeName=upperarm_l_capsule, center=<20 -5 -4>, rotation=<rot 60 -30 0>}]
 upperarm_r [{shapeType=Capsule, length=32, radius=8, shapeName=upperarm_r_capsule, center=<-20 5 4>, rotation=<rot 60 -40 0>}]
}
constraints [
 {bone1Name=dyn_front_dread_parent, bone2Name=dyn_front_dread_base, swing1LimitDegrees=20, swing2LimitDegrees=30, twistLimitDegrees=3}
 {bone1Name=dyn_front_dread_01, bone2Name=dyn_front_dread_parent, swing1LimitDegrees=15, swing2LimitDegrees=35, twistLimitDegrees=3}
 {bone1Name=dyn_front_dread_02, bone2Name=dyn_front_dread_01, swing1LimitDegrees=10, swing2LimitDegrees=35, twistLimitDegrees=3}
 {bone1Name=dyn_front_dread_03, bone2Name=dyn_front_dread_02, swing1LimitDegrees=20, swing2LimitDegrees=30, twistLimitDegrees=3}
 {bone1Name=dyn_dread_parent_a, bone2Name=dyn_dread_base_a, swing1LimitDegrees=8, swing2LimitDegrees=30, twistLimitDegrees=3}
 {bone1Name=dyn_dread_a_01, bone2Name=dyn_dread_parent_a, swing1LimitDegrees=4, swing2LimitDegrees=35, twistLimitDegrees=3}
 {bone1Name=dyn_dread_a_02, bone2Name=dyn_dread_a_01, swing1LimitDegrees=4, swing2LimitDegrees=35, twistLimitDegrees=3}
 {bone1Name=dyn_dread_parent_b, bone2Name=dyn_dread_base_b, swing1LimitDegrees=17, swing2LimitDegrees=40, twistLimitDegrees=3}
 {bone1Name=dyn_dread_b_01, bone2Name=dyn_dread_parent_b, swing1LimitDegrees=17, swing2LimitDegrees=35, twistLimitDegrees=3}
 {bone1Name=dyn_dread_b_02, bone2Name=dyn_dread_b_01, swing1LimitDegrees=17, swing2LimitDegrees=35, twistLimitDegrees=3}
 {bone1Name=dyn_dread_parent_c, bone2Name=dyn_dread_base_c, swing1LimitDegrees=5, swing2LimitDegrees=60, twistLimitDegrees=3}
 {bone1Name=dyn_dread_c_01, bone2Name=dyn_dread_parent_c, swing1LimitDegrees=5, swing2LimitDegrees=60, twistLimitDegrees=3}
 {bone1Name=dyn_dread_c_02, bone2Name=dyn_dread_c_01, swing1LimitDegrees=5, swing2LimitDegrees=60, twistLimitDegrees=3}
 {bone1Name=dyn_dread_parent_d, bone2Name=dyn_dread_base_d, swing1LimitDegrees=8, swing2LimitDegrees=40, twistLimitDegrees=3}
 {bone1Name=dyn_dread_d_01, bone2Name=dyn_dread_parent_d, swing1LimitDegrees=8, swing2LimitDegrees=35, twistLimitDegrees=3}
 {bone1Name=dyn_dread_d_02, bone2Name=dyn_dread_d_01, swing1LimitDegrees=8, swing2LimitDegrees=35, twistLimitDegrees=3}
 {bone1Name=dyn_dread_parent_e, bone2Name=dyn_dread_base_e, swing1LimitDegrees=12, swing2LimitDegrees=30, twistLimitDegrees=3}
 {bone1Name=dyn_dread_e_02, bone2Name=dyn_dread_e_01, swing1LimitDegrees=12, swing2LimitDegrees=20, twistLimitDegrees=3}
 {bone1Name=dyn_dread_e_01, bone2Name=dyn_dread_parent_e, swing1LimitDegrees=12, swing2LimitDegrees=20, twistLimitDegrees=3}
 {bone1Name=dyn_dread_parent_f, bone2Name=dyn_dread_base_f, swing1LimitDegrees=8, swing2LimitDegrees=30, twistLimitDegrees=3}
 {bone1Name=dyn_dread_f_01, bone2Name=dyn_dread_parent_f, swing1LimitDegrees=8, swing2LimitDegrees=35, twistLimitDegrees=3}
 {bone1Name=dyn_dread_f_02, bone2Name=dyn_dread_f_01, swing1LimitDegrees=8, swing2LimitDegrees=35, twistLimitDegrees=3}
 {bone1Name=dyn_dread_parent_g, bone2Name=dyn_dread_base_g, swing1LimitDegrees=10, swing2LimitDegrees=45, twistLimitDegrees=3}
 {bone1Name=dyn_dread_g_01, bone2Name=dyn_dread_parent_g, swing1LimitDegrees=10, swing2LimitDegrees=20, twistLimitDegrees=3}
 {bone1Name=dyn_dread_g_02, bone2Name=dyn_dread_g_01, swing1LimitDegrees=10, swing2LimitDegrees=20, twistLimitDegrees=3}
 {bone1Name=dyn_dread_parent_h, bone2Name=dyn_dread_base_h, swing1LimitDegrees=8, swing2LimitDegrees=15, twistLimitDegrees=3}
 {bone1Name=dyn_dread_h_01, bone2Name=dyn_dread_parent_h, swing1LimitDegrees=8, swing2LimitDegrees=35, twistLimitDegrees=3}
 {bone1Name=dyn_dread_h_02, bone2Name=dyn_dread_h_01, swing1LimitDegrees=8, swing2LimitDegrees=35, twistLimitDegrees=3}
 {bone1Name=dyn_dread_a_03, bone2Name=dyn_dread_a_02, swing1LimitDegrees=4, swing2LimitDegrees=35, twistLimitDegrees=3}
 {bone1Name=dyn_dread_d_03, bone2Name=dyn_dread_d_02, swing1LimitDegrees=8, swing2LimitDegrees=35, twistLimitDegrees=3}
 {bone1Name=dyn_dread_h_03, bone2Name=dyn_dread_h_02, swing1LimitDegrees=8, swing2LimitDegrees=35, twistLimitDegrees=3}
 {bone1Name=dyn_dread_e_03, bone2Name=dyn_dread_e_02, swing1LimitDegrees=12, swing2LimitDegrees=20, twistLimitDegrees=3}
 {
  bone1Name=upperarm_l
  bone2Name=spine_05
  swing1LimitDegrees=45
  swing1Motion=Free
  swing2LimitDegrees=45
  swing2Motion=Free
  twistLimitDegrees=45
  twistMotion=Free
 }
 {
  bone1Name=upperarm_r
  bone2Name=clavicle_r
  swing1LimitDegrees=45
  swing1Motion=Free
  swing2LimitDegrees=45
  swing2Motion=Free
  twistLimitDegrees=45
  twistMotion=Free
 }
]
