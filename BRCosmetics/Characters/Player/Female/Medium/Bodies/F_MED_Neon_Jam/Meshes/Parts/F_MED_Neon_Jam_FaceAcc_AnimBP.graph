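# Animation Blueprint for the Neon Jam face accessory, posing and simulating its dreadlock bones on top of the pose copied from the parent mesh.
# A cached pose with front-end bone offsets feeds separate RigidBody branches for front end, no backpack, backpack and cape, chosen by bool blends.
# Backpack and cape branches run Control Rigs first; the no-backpack branch adds skydive bone offsets.
# The event graph derives skydive alpha, external force and gravity override from pawn state.
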AnimBlueprint /BRCosmetics/Characters/Player/Female/Medium/Bodies/F_MED_Neon_Jam/Meshes/Parts/F_MED_Neon_Jam_FaceAcc_AnimBP
target_skeleton=&/F_MED_Neon_Jam_FaceAcc_Skeleton
graphs [
 AnimGraph {nodes [
   "Transform (Modify) Bone - Bone: head_offset"(Alpha=1, Translation="0,0,0", Rotation="-20,0,0", Scale="1,1,1") {
    rotationMode=BMM_Additive
    rotationSpace=BCS_BoneSpace
    alphaBoolBlend {blendOption=Linear}
    boneToModify {boneName=head_offset}
    rotation {pitch=-20}
   }
   n1 = "Transform (Modify) Bone - Bone: dyn_dread_base_f"(Alpha=1, Translation="0,0,0", Rotation="-15,-10,10", Scale="1,1,1") {
    rotationMode=BMM_Additive
    rotationSpace=BCS_BoneSpace
    alphaBoolBlend {blendOption=Linear}
    boneToModify {boneName=dyn_dread_base_f}
    rotation=<rot -15 -10 10>
   }
   "Transform (Modify) Bone - Bone: dyn_dread_base_a"(ComponentPose=@n1, Alpha=1, Translation="0,0,0", Rotation="0,-6,0", Scale="1,1,1") {
    rotationMode=BMM_Additive
    rotationSpace=BCS_BoneSpace
    alphaBoolBlend {blendOption=Linear}
    boneToModify {boneName=dyn_dread_base_a}
    rotation {yaw=-6}
   }
   n2 = Get Is_Skydiving() -> (Is_Skydiving:double)
   n3 = "Transform (Modify) Bone - Bone: head_offset"(
    ComponentPose=Local To Component(LocalPose=Use cached pose 'BackpackCache'())
    Alpha=@n2
    Translation="0,0,0"
    Rotation="0,10,0"
    Scale="1,1,1"
   ) {
    rotationMode=BMM_Additive
    rotationSpace=BCS_BoneSpace
    alphaBoolBlend {blendOption=Linear}
    boneToModify {boneName=head_offset}
    rotation {yaw=10}
   }
   n4 = "Transform (Modify) Bone - Bone: dyn_dread_base_a"(ComponentPose=@n3, Alpha=@n2, Translation="0,0,0", Rotation="5,-10,0", Scale="1,1,1") {
    rotationMode=BMM_Additive
    rotationSpace=BCS_BoneSpace
    alphaBoolBlend {blendOption=Linear}
    boneToModify {boneName=dyn_dread_base_a}
    rotation {pitch=5, yaw=-10}
   }
   n5 = "Transform (Modify) Bone - Bone: dyn_dread_base_b"(ComponentPose=@n4, Alpha=@n2, Translation="0,0,0", Rotation="-50,0,0", Scale="1,1,1") {
    rotationMode=BMM_Additive
    rotationSpace=BCS_BoneSpace
    alphaBoolBlend {blendOption=Linear}
    boneToModify {boneName=dyn_dread_base_b}
    rotation {pitch=-50}
   }
   n6 = "Transform (Modify) Bone - Bone: dyn_dread_base_f"(ComponentPose=@n5, Alpha=@n2, Translation="0,0,0", Rotation="-50,0,0", Scale="1,1,1") {
    rotationMode=BMM_Additive
    rotationSpace=BCS_BoneSpace
    alphaBoolBlend {blendOption=Linear}
    boneToModify {boneName=dyn_dread_base_f}
    rotation {pitch=-50}
   }
   n7 = "Transform (Modify) Bone - Bone: dyn_dread_base_h"(ComponentPose=@n6, Alpha=@n2, Translation="0,0,0", Rotation="-50,0,0", Scale="1,1,1") {
    rotationMode=BMM_Additive
    rotationSpace=BCS_BoneSpace
    alphaBoolBlend {blendOption=Linear}
    boneToModify {boneName=dyn_dread_base_h}
    rotation {pitch=-50}
   }
   n8 = "Transform (Modify) Bone - Bone: dyn_dread_base_g"(ComponentPose=@n7, Alpha=@n2, Translation="0,0,0", Rotation="-50,0,0", Scale="1,1,1") {
    rotationMode=BMM_Additive
    rotationSpace=BCS_BoneSpace
    alphaBoolBlend {blendOption=Linear}
    boneToModify {boneName=dyn_dread_base_g}
    rotation {pitch=-50}
   }
   n9 = "Transform (Modify) Bone - Bone: dyn_dread_base_e"(ComponentPose=@n8, Alpha=@n2, Translation="0,0,0", Rotation="-10,10,0", Scale="1,1,1") {
    rotationMode=BMM_Additive
    rotationSpace=BCS_BoneSpace
    alphaBoolBlend {blendOption=Linear}
    boneToModify {boneName=dyn_dread_base_e}
    rotation {pitch=-10, yaw=10}
   }
   n10 = "Transform (Modify) Bone - Bone: dyn_dread_base_d"(ComponentPose=@n9, Alpha=@n2, Translation="0,0,0", Rotation="20,0,0", Scale="1,1,1") {
    rotationMode=BMM_Additive
    rotationSpace=BCS_BoneSpace
    alphaBoolBlend {blendOption=Linear}
    boneToModify {boneName=dyn_dread_base_d}
    rotation {pitch=20}
   }
   n11 = "Transform (Modify) Bone - Bone: dyn_dread_base_c"(ComponentPose=@n10, Alpha=@n2, Translation="0,0,0", Rotation="-15,-15,0", Scale="1,1,1") {
    rotationMode=BMM_Additive
    rotationSpace=BCS_BoneSpace
    alphaBoolBlend {blendOption=Linear}
    boneToModify {boneName=dyn_dread_base_c}
    rotation {pitch=-15, yaw=-15}
   }
   n12 = RigidBody!(
    ComponentPose=@n11
    Alpha=Get Alpha_Skydive() -> (Alpha_Skydive:double)
    OverrideWorldGravity:Vector=Get Gravity_Override() -> (Gravity_Override:Vector)
    ExternalForce=Get Skydiving_Velocity() -> (Skydiving_Velocity:Vector)
   ) -> (Pose) {
    bClampLinearTranslationLimitToRefPose=true
    bOverrideWorldGravity=true
    overridePhysicsAsset=&/F_MED_Neon_Jam_FaceAcc_Physics
    simulationSpace=BaseBoneSpace
    alphaBoolBlend {blendOption=Linear}
    baseBoneRef {boneName=neck_01}
    componentAppliedLinearAccClamp=<6000 6000 6000>
    componentLinearAccScale=<0.1 0.2 0.1>
    componentLinearVelScale=<0.1 0.2 0.1>
    overrideWorldGravity {z=-980}
    simSpaceSettings {
     maxAngularAcceleration=100
     maxAngularVelocity=3
     maxLinearAcceleration=400
     maxLinearVelocity=250
     velocityScaleZ=0
     worldAlpha=1
     externalLinearDragV=<0.1 0.1>
    }
   }
   n13 = Component To Local(ComponentPose=@n12)
   n14 = Get isWearingPack() -> (isWearingPack:bool)
   n15 = Control Rig!(Source=Use cached pose 'BackpackCache'(), Alpha, isWearingPack:bool=false) -> (Pose) {
    alphaBoolBlend {blendOption=Linear}
    controlRigAssetReference {blueprintRigClass=&/F_MED_Neon_Jam_FaceAcc_CtrlRig.F_MED_Neon_Jam_FaceAcc_CtrlRig_C}
   }
   n16 = RigidBody!(ComponentPose=Local To Component(LocalPose=@n15), Alpha, ExternalForce) -> (Pose) {
    bClampLinearTranslationLimitToRefPose=true
    bOverrideWorldGravity=true
    overridePhysicsAsset=&/F_MED_Neon_Jam_Pack_FaceAcc_Physics
    simulationSpace=BaseBoneSpace
    alphaBoolBlend {blendOption=Linear}
    baseBoneRef {boneName=neck_02}
    componentAppliedLinearAccClamp=<6000 6000 6000>
    componentLinearAccScale=<0.1 0.1 0.1>
    componentLinearVelScale=<0.1 0.1 0.1>
    overrideWorldGravity {z=-980}
    simSpaceSettings {
     maxAngularAcceleration=100
     maxAngularVelocity=3
     maxLinearAcceleration=400
     maxLinearVelocity=250
     velocityScaleZ=0
     worldAlpha=1
     externalLinearDragV=<0.1 0.1>
    }
   }
   n17 = Control Rig(Source=Use cached pose 'BackpackCache'(), isWearingPack=false) {
    alphaBoolBlend {blendOption=Linear}
    controlRigAssetReference {blueprintRigClass=&/F_MED_Neon_Jam_Cape_FaceAcc_CtrlRig.F_MED_Neon_Jam_Cape_FaceAcc_CtrlRig_C}
   }
   n18 = RigidBody!(
    ComponentPose=Local To Component(LocalPose=@n17)
    Alpha=Get Alpha_Skydive()
    OverrideWorldGravity:Vector=Get Gravity_Override()
    ExternalForce=Get Skydiving_Velocity()
   ) -> (Pose) {
    bClampLinearTranslationLimitToRefPose=true
    bOverrideWorldGravity=true
    overridePhysicsAsset=&/F_MED_Neon_Jam_Cape_FaceAcc_Physics
    simulationSpace=BaseBoneSpace
    alphaBoolBlend {blendOption=Linear}
    baseBoneRef {boneName=neck_01}
    componentAppliedLinearAccClamp=<6000 6000 6000>
    componentLinearAccScale=<0.1 0.2 0.1>
    componentLinearVelScale=<0.1 0.2 0.1>
    overrideWorldGravity {z=-980}
    simSpaceSettings {
     maxAngularAcceleration=100
     maxAngularVelocity=3
     maxLinearAcceleration=400
     maxLinearVelocity=250
     velocityScaleZ=0
     worldAlpha=1
     externalLinearDragV=<0.1 0.1>
    }
   }
   n19 = Blend Poses by bool(
    BlendPose_0=Component To Local(ComponentPose=@n18)
    BlendPose_1=Component To Local(ComponentPose=@n16)
    bActiveValue=Get isWearingCape() -> (isWearingCape:bool)
   ) {blendType=Linear}
   n20 = RigidBody(ComponentPose=Local To Component(LocalPose=Use cached pose 'BackpackCache'()), OverrideWorldGravity="0,0,-980") {
    bClampLinearTranslationLimitToRefPose=true
    bOverrideWorldGravity=true
    overridePhysicsAsset=&/F_MED_Neon_Jam_FrontEnd_FaceAcc_Physics
    simulationSpace=BaseBoneSpace
    alphaBoolBlend {blendOption=Linear}
    baseBoneRef {boneName=neck_01}
    componentAppliedLinearAccClamp=<6000 6000 6000>
    componentLinearAccScale=<0.1 0.2 0.1>
    componentLinearVelScale=<0.1 0.2 0.1>
    overrideWorldGravity {z=-980}
    simSpaceSettings {
     maxAngularAcceleration=100
     maxAngularVelocity=3
     maxLinearAcceleration=400
     maxLinearVelocity=250
     velocityScaleZ=0
     worldAlpha=1
     externalLinearDragV=<0.1 0.1>
    }
   }
   Output Pose(Result=Blend Poses by bool(
     BlendPose_0=Component To Local(ComponentPose=@n20)
     BlendPose_1=Blend Poses by bool(BlendPose_0=@n19, BlendPose_1=@n13, bActiveValue=@n14) {blendType=Linear}
     bActiveValue=Get bIsInFrontEnd()
    ) {blendType=Linear}) {layerGroup=None}
   n21 = Get FrontEndCheck() -> (FrontEndCheck:double)
   n22 = "Transform (Modify) Bone - Bone: dyn_dread_base_h"(
    ComponentPose=Local To Component(LocalPose=Copy Pose From Mesh() {bUseAttachedParent=true})
    Alpha=@n21
    Translation="0,0,0"
    Rotation="-20,0,15"
    Scale="1,1,1"
   ) {
    rotationMode=BMM_Additive
    rotationSpace=BCS_BoneSpace
    alphaBoolBlend {blendOption=Linear}
    boneToModify {boneName=dyn_dread_base_h}
    rotation {pitch=-20, roll=15}
   }
   n23 = "Transform (Modify) Bone - Bone: dyn_dread_base_e"(ComponentPose=@n22, Alpha=@n21, Translation="0,0,0", Rotation="-23.337,6.679,-3.855", Scale="1,1,1") {
    rotationMode=BMM_Additive
    rotationSpace=BCS_BoneSpace
    alphaBoolBlend {blendOption=Linear}
    boneToModify {boneName=dyn_dread_base_e}
    rotation=<rot -23.337 6.679 -3.855>
   }
   n24 = "Transform (Modify) Bone - Bone: dyn_dread_base_d"(ComponentPose=@n23, Alpha=@n21, Translation="0,0,0", Rotation="-40,0,12", Scale="1,1,1") {
    rotationMode=BMM_Additive
    rotationSpace=BCS_BoneSpace
    alphaBoolBlend {blendOption=Linear}
    boneToModify {boneName=dyn_dread_base_d}
    rotation {pitch=-40, roll=12}
   }
   n25 = "Transform (Modify) Bone - Bone: dyn_dread_base_b"(ComponentPose=@n24, Alpha=@n21, Translation="0,0,0", Rotation="-10,0,-10", Scale="1,1,1") {
    rotationMode=BMM_Additive
    rotationSpace=BCS_BoneSpace
    alphaBoolBlend {blendOption=Linear}
    boneToModify {boneName=dyn_dread_base_b}
    rotation {pitch=-10, roll=-10}
   }
   Save cached pose 'BackpackCache'(Pose=Component To Local(ComponentPose=@n25)) {cacheName=BackpackCache}
   Get Skydiving_Velocity()
  ]}
 EventGraph {nodes [
   Event BlueprintInitializeAnimation()
   > n1 = Cast To PlayerPawn_Athena(Object=TryGetPawnOwner())
   > n2 = Set Pawn Owner(Pawn Owner:Player Pawn Athena Object Reference=@n1) -> (Output_Get:Player Pawn Athena Object Reference)
   > Set Body_Mesh(Body_Mesh=GetSkeletalMeshForPartType(self=@n2))
   Event BlueprintUpdateAnimation()
   > n3 = Sequence!(execute) -> (then_0, then_1, then_2:Exec, then_3:Exec, then_4:Exec, then_5:Exec)
   n4 = Get bIsSkydiveFloating()
   n5 = Get bIsSkyDiving()
   n6 = Get bIsSkydiveDiveMode()
   > Set Skydiving_Velocity(Skydiving_Velocity:Vector=SelectVector(A="0,0,40000", B="0, 0, 0", bPickA=OR Boolean!(A=@n4, B=@n5, C:bool=@n6) -> (ReturnValue))) -> (Output_Get:Vector)
   n7 = Get bIsSkydiveFloating()
   n8 = Get bIsSkyDiving()
   n9 = Get bIsSkydiveDiveMode()
   Set Alpha_Skydive(execute=@n3.then_2, Alpha_Skydive:double=SelectFloat(A=0.8, B=1, bPickA=OR Boolean(A=@n7, B=@n8, C=@n9))) -> (Output_Get:double)
   n10 = Get bIsSkydiveFloating()
   n11 = Get bIsSkyDiving()
   n12 = Get bIsSkydiveDiveMode()
   Set Is_Skydiving(execute=@n3.then_1, Is_Skydiving:double=SelectFloat(A=1, B=0, bPickA=OR Boolean(A=@n10, B=@n11, C=@n12))) -> (Output_Get:double)
   n13 = Gravity Override State Function DEPRECATED(
    execute=@n3.then_3
    CharacterPartAnimInstance=Self-Reference()
    GameTimeInSeconds=GetGameTimeInSeconds()
    GravityOverrideParams=Get Gravity Override Params()
   )
   > Set Gravity_Override(Gravity_Override:Vector=@n13) -> (Output_Get:Vector)
   Set FrontEndCheck(execute=@n3.then_4, FrontEndCheck:double=SelectFloat(A=1, B=0, bPickA=Get bIsInFrontEnd())) -> (Output_Get:double)
   Event OnCharacterCustomizationLoaded()
   > Delay(Duration=0)
   > ResetDynamics(InTeleportType=TeleportPhysics)
   > ResetDynamics(InTeleportType=ResetPhysics)
   n14 = CheckForBackbling()
   Set isWearingPack(isWearingPack:bool=@n14) -> (Output_Get:bool)
   > Set isWearingCape(isWearingCape=AND Boolean(A=@n14, B=@n14.bIsBackblingLong))
  ]}
]
variables [
 Alpha_Skydive {type=number}
 Body_Mesh {type=SkeletalMeshComponent}
 FrontEndCheck {type=number}
 Gravity Override Params {
  type=GravityOverrideParamsStruct
  default {
   skydive_DivingDown {windFrequency=<7 18 12>}
   skydive_FloatingDown {windFrequency=<8 12 12>}
   skydive_GoingUp {windFrequency=<8 12 12>}
   skydive_Parachuting {gravityOverride=<0 0 -980>, windAmplitude=<30 40 50>, windFrequency=<14 14 10>}
  }
 }
 Gravity_Override {type=Vector}
 Is_Skydiving {type=number}
 isWearingCape {type=boolean}
 isWearingPack {type=boolean}
 JumpForce {type=Vector}
 Pawn Owner {type=PlayerPawn_Athena_C}
 Skel_Mesh {type=SkeletalMeshComponent}
 Skydiving_Velocity {type=Vector}
]
